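; Building actor Blueprint for a large leaking pipe prop, with a Niagara leak effect and looping spout and splash audio.
; The construction script feeds the SplashHeight variable as a Z offset into the leak effect's SplashPosition parameter.
; On death, it briefly fades out both audio loops after validity checks.
; BeginPlay, overlap and Tick events are empty.
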(event EventOnDeathPlayEffects
  (params (Damage :type Float%20%28single-precision%29) (DamageTags :type Gameplay%20Tag%20Container%20Structure%20%28by%20ref%29) (Momentum :type Vector :default "0, 0, 0") (HitInfo :type Hit%20Result%20Structure%20%28by%20ref%29) (InstigatedBy :type Fort%20Pawn%20Object%20Reference) (DamageCauser :type Actor%20Object%20Reference) (EffectContext :type Gameplay%20Effect%20Context%20Handle%20Structure))
  (bind _audio_pipe_leak_spout_water_loop (Variables|Default|GetAudio_Pipe_Leak_Spout_Water_Loop))
  (bind _audio_pipe_leak_water_splashes_loop (Variables|Default|GetAudio_Pipe_Leak_Water_Splashes_Loop))
  ;[1 1 1] On Death Stop Audio
  (Utilities|FlowControl|Sequence
    (:then_0
      (Utilities|IsValid _audio_pipe_leak_spout_water_loop
        (:Is%20Valid
          (if (Utilities|IsValid _audio_pipe_leak_spout_water_loop :!class /Script/Engine.KismetSystemLibrary)
            (Audio|Components|Audio|FadeOut _audio_pipe_leak_spout_water_loop 0.2 0.0 "Linear" :!class /Script/Engine.AudioComponent)))))
    (:then_1
      (Utilities|IsValid _audio_pipe_leak_water_splashes_loop
        (:Is%20Valid
          (if (Utilities|IsValid _audio_pipe_leak_water_splashes_loop :!class /Script/Engine.KismetSystemLibrary)
            (Audio|Components|Audio|FadeOut _audio_pipe_leak_water_splashes_loop 0.2 0.0 "Linear" :!class /Script/Engine.AudioComponent)))))))

(event EventBeginPlay)

(event Collision|EventActorBeginOverlap
  (params (OtherActor :type Actor%20Object%20Reference)))

(event EventTick
  (params (DeltaSeconds :type Float%20%28single-precision%29)))

(fn ConstructionScript
  (Niagara|SetNiagaraVariable(Vector3) (Variables|Default|GetLeakEffect) "SplashPosition" (Math|Vector|MakeVector 0.0 0.0 (Variables|Default|GetSplashHeight))))
